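# Sound cue for a llama disappearing in the Bling limited-time mode, built from reused teleport sounds.
# A mixer layers one of three randomly picked dungeon teleport fire variations with a quieter, pitched-up portal teleport sound.
# An outer modulator adds slight pitch variation to the combined result.
# It plays through a dedicated attenuation asset and a third-person ranged weapon fire sound class.
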SoundCue /Game/Athena/Sounds/LTMs/Bling/Bling_Llama_Disappear_Cue
root=Modulator_2
nodes [
 {id=Mixer_0, children [Random_1,Modulator_1], properties {inputVolume [1,0.65]}}
 {id=Modulator_1, children [WavePlayer_6], properties {pitchMax=1.25, pitchMin=1.2}}
 {id=Modulator_2, children [Mixer_0], properties {pitchMax=1.02, pitchMin=0.98}}
 {id=Random_1, children [WavePlayer_3,WavePlayer_4,WavePlayer_5], properties {weights [1,1,1]}}
 {id=WavePlayer_3, properties {soundWaveAssetPtr=/Game/Sounds/Fort_Traps/Dungeon_Traps/Dungeon_Teleport_Fire_01}}
 {id=WavePlayer_4, properties {soundWaveAssetPtr=/Game/Sounds/Fort_Traps/Dungeon_Traps/Dungeon_Teleport_Fire_02}}
 {id=WavePlayer_5, properties {soundWaveAssetPtr=/Game/Sounds/Fort_Traps/Dungeon_Traps/Dungeon_Teleport_Fire_03}}
 {id=WavePlayer_6, properties {soundWaveAssetPtr=/Game/Sounds/Fort_GamePlay_Sounds/Player/Portal/Portal_Teleport_01}}
]
properties {
 attenuationSettings=&/Bling_Llama_Disappear_Att
 maxDistance=4500
 soundClassObject=&Fort_WeaponRanged_Fire_3P_PSM
}
