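# Sound cue for a llama drop in the Bling limited-time mode, built by layering two reused sounds.
# A Mixer blends a jetpack drop wave, pitched up, with a vending machine sparkle wave, pitched down.
# Each layer's Modulator locks its pitch to a fixed value, so playback does not vary.
# The cue plays at reduced volume through the LTM gadgets sound class with weapon-drop attenuation.
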SoundCue /Game/Athena/Sounds/LTMs/Bling/Bling_Llama_Drop_Cue
root=Mixer_0
nodes [
 {id=Mixer_0, children [Modulator_0,Modulator_1], properties {inputVolume [1,1]}}
 {id=Modulator_0, children [WavePlayer_0], properties {pitchMax=1.5, pitchMin=1.5}}
 {id=Modulator_1, children [WavePlayer_1], properties {pitchMax=0.8, pitchMin=0.8}}
 {id=WavePlayer_0, properties {soundWaveAssetPtr=/Game/Athena/Sounds/Items/Jetpack/Jetpack_Drop_01}}
 {id=WavePlayer_1, properties {soundWaveAssetPtr=/Game/Sounds/Foley_Loot/Containers/VendingMachine/VendingMachine_Vend_Sparkle}}
]
properties {
 attenuationSettings=&Fort_Weapon_Drop_Att
 maxDistance=3000
 soundClassObject=/Game/Sounds/FortSoundClassesAndModes/Fort_Gadgets_LTM
 volumeMultiplier=0.4
}
