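# Sound cue for the looping in-world ambience of the llama in the Bling limited-time mode.
# A mixer layers two looping waves, the llama in-world loop and a Bling backpack loop, at equal input volume.
# It plays through the loot ambient loop sound class, with overridden attenuation using logarithmic falloff and listener focus that quiets it when out of focus.
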SoundCue /Game/Athena/Sounds/LTMs/Bling/Bling_Llama_InWorldLoop_Cue
root=Mixer_0
nodes [
 {id=Mixer_0, children [WavePlayer_0,WavePlayer_1], properties {inputVolume [0.8,0.8]}}
 {id=WavePlayer_0, properties {bLooping=true, soundWaveAssetPtr=&/Bling_Llama_InWorldLoop_01}}
 {id=WavePlayer_1, properties {bLooping=true, soundWaveAssetPtr=&/Bling_BackPack_Loop}}
]
properties {
 bOverrideAttenuation=true
 maxDistance=900
 soundClassObject=/Game/Sounds/FortSoundClassesAndModes/Fort_Loot_Ambient_Loop
 volumeMultiplier=1.3
 attenuationOverrides {
  bEnableListenerFocus=true
  distanceAlgorithm=Logarithmic
  falloffDistance=700
  nonFocusVolumeAttenuation=0.4762
  stereoSpread=35.6359
  attenuationShapeExtents {x=200}
 }
}
