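# Sound cue for the Bling Llama pickup in a limited-time mode, built from two layered samples.
# A mixer plays a jetpack pickup wave, pitched up by a fixed amount through a modulator, alongside a Blink pickup layer at normal pitch.
# The cue plays at reduced volume through the LTM gadgets sound class with weapon-drop attenuation.
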SoundCue /Game/Athena/Sounds/LTMs/Bling/Bling_Llama_Pickup_Cue
root=Mixer_0
nodes [
 {id=Mixer_0, children [Modulator_0,Modulator_1], properties {inputVolume [1,1]}}
 {id=Modulator_0, children [WavePlayer_0], properties {pitchMax=1.5, pitchMin=1.5}}
 {id=Modulator_1, children [WavePlayer_2], properties {pitchMax=1, pitchMin=1}}
 {id=WavePlayer_0, properties {soundWaveAssetPtr=/Game/Athena/Sounds/Items/Jetpack/Jetpack_Pickup_01}}
 {id=WavePlayer_2, properties {soundWaveAssetPtr=&/Blink_Pickup_Layer}}
]
properties {
 attenuationSettings=&Fort_Weapon_Drop_Att
 maxDistance=3000
 soundClassObject=/Game/Sounds/FortSoundClassesAndModes/Fort_Gadgets_LTM
 volumeMultiplier=0.4
}
